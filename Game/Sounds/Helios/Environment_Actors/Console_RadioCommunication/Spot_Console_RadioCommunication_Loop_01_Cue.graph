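# Sound cue for a looping spot ambience of walkie-talkie radio chatter on the Helios console radio communication prop.
# One of six equally weighted chatter waves is picked at random, pitch-modulated slightly, then played after a random delay inside a loop.
# A dedicated attenuation asset wraps the loop, routed to the Athena ambient props sound class at low volume.
# A concurrency override caps simultaneous instances and blocks new ones.
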SoundCue /Game/Sounds/Helios/Environment_Actors/Console_RadioCommunication/Spot_Console_RadioCommunication_Loop_01_Cue
root=Attenuation_0
nodes [
 {
  id=Attenuation_0
  children [Looping_0]
  properties {attenuationSettings=&/Spot_Console_RadioCommunication_Loop_Att}
 }
 {id=Delay_0, children [Modulator_0], properties {delayMax=3}}
 {id=Looping_0, children [Delay_0]}
 {id=Modulator_0, children [Random_0], properties {pitchMax=1.1, pitchMin=0.9}}
 {id=Random_0, children [WavePlayer_1,WavePlayer_2,WavePlayer_3,WavePlayer_4,WavePlayer_5,WavePlayer_6], properties {weights [1,1,1,1,1,1]}}
 {
  id=WavePlayer_1
  properties {soundWaveAssetPtr=&/Spot_Radio_Communication_Walkie_Talkie_01}
 }
 {
  id=WavePlayer_2
  properties {soundWaveAssetPtr=&/Spot_Radio_Communication_Walkie_Talkie_02}
 }
 {
  id=WavePlayer_3
  properties {soundWaveAssetPtr=&/Spot_Radio_Communication_Walkie_Talkie_03}
 }
 {
  id=WavePlayer_4
  properties {soundWaveAssetPtr=&/Spot_Radio_Communication_Walkie_Talkie_04}
 }
 {
  id=WavePlayer_5
  properties {soundWaveAssetPtr=&/Spot_Radio_Communication_Walkie_Talkie_05}
 }
 {
  id=WavePlayer_6
  properties {soundWaveAssetPtr=&/Spot_Radio_Communication_Walkie_Talkie_06}
 }
]
properties {
 bOverrideConcurrency=true
 maxDistance=850
 soundClassObject=&Fort_Ambient_Props_Athena
 volumeMultiplier=0.125
 concurrencyOverrides {maxCount=4, resolutionRule=PreventNew, volumeScale=0.5}
}
