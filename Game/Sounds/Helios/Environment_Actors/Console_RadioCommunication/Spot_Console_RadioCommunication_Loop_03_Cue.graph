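# Sound cue for a looping spot ambience of radio tuning at a radio communication console, in the ambient props sound class.
# A looping node repeatedly picks one of three equally weighted tuning waves, adding a random delay and pitch modulation to vary each pass.
# A dedicated attenuation asset wraps the chain.
# A concurrency override caps simultaneous instances and prevents new ones beyond the limit.
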SoundCue /Game/Sounds/Helios/Environment_Actors/Console_RadioCommunication/Spot_Console_RadioCommunication_Loop_03_Cue
root=Attenuation_0
nodes [
 {
  id=Attenuation_0
  children [Looping_0]
  properties {attenuationSettings=&/Spot_Console_RadioCommunication_Loop_Att}
 }
 {id=Delay_0, children [Modulator_0], properties {delayMax=3}}
 {id=Looping_0, children [Delay_0]}
 {id=Modulator_0, children [Random_1], properties {pitchMin=0.8}}
 {id=Random_1, children [WavePlayer_7,WavePlayer_8,WavePlayer_9], properties {weights [1,1,1]}}
 {
  id=WavePlayer_7
  properties {soundWaveAssetPtr=&/Spot_Radio_Communcation_Tuning_Loop_01}
 }
 {
  id=WavePlayer_8
  properties {soundWaveAssetPtr=&/Spot_Radio_Communcation_Tuning_Loop_02}
 }
 {
  id=WavePlayer_9
  properties {soundWaveAssetPtr=&/Spot_Radio_Communcation_Tuning_Loop_03}
 }
]
properties {
 bOverrideConcurrency=true
 maxDistance=850
 soundClassObject=&Fort_Ambient_Props_Athena
 volumeMultiplier=0.2
 concurrencyOverrides {maxCount=4, resolutionRule=PreventNew, volumeScale=0.5}
}
