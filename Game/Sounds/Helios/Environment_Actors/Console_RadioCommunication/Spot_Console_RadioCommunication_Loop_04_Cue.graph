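# Spatialized ambient sound cue for a radio communication console prop, looping intermittent walkie-talkie chatter.
# Each loop pass waits a random delay of up to a few seconds, then plays one of three equally weighted waves with slight pitch variation.
# A shared attenuation asset wraps the loop, and output feeds an exterior ambient source bus and the default reverb submix.
# A concurrency override caps simultaneous instances at two.
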SoundCue /Game/Sounds/Helios/Environment_Actors/Console_RadioCommunication/Spot_Console_RadioCommunication_Loop_04_Cue
root=Attenuation_0
nodes [
 {
  id=Attenuation_0
  children [Looping_0]
  properties {attenuationSettings=&/Spot_Console_RadioCommunication_Loop_Att}
 }
 {id=Delay_0, children [Modulator_0], properties {delayMax=3}}
 {id=Looping_0, children [Delay_0]}
 {id=Modulator_0, children [Random_0], properties {pitchMax=1.1, pitchMin=0.9}}
 {id=Random_0, children [WavePlayer_0,WavePlayer_2,WavePlayer_3], properties {weights [1,1,1]}}
 {
  id=WavePlayer_0
  properties {soundWaveAssetPtr=&/Spot_Radio_Communication_Walkie_Talkie_01}
 }
 {
  id=WavePlayer_2
  properties {soundWaveAssetPtr=&/Spot_Radio_Communication_Walkie_Talkie_02}
 }
 {
  id=WavePlayer_3
  properties {soundWaveAssetPtr=&/Spot_Radio_Communication_Walkie_Talkie_03}
 }
]
properties {
 bOverrideConcurrency=true
 maxDistance=850
 soundClassObject=&Fort_Ambient_Props_Athena
 volumeMultiplier=0.2
 busSends [{
   audioBus=null
   minSendLevel=0.25
   sendLevel=1
   soundSourceBus=&AmbientAudio_Exterior_SourceBus
   customSendLevelCurve {editorCurveData {keys [<at 0 0.25>,<at 1 1>]}}
  }]
 concurrencyOverrides {maxCount=2, volumeScale=0.5}
 soundSubmixSends [{
   minSendLevel=0.25
   soundSubmix=&Fortnite_Reverb_Default
   customSendLevelCurve {editorCurveData {keys [<at 0 0.25>,<at 1 1>]}}
  }]
}
